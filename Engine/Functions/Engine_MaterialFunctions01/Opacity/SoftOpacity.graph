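# Material function in the Opacity library that softens an opacity value by camera proximity and viewing angle.
# An inverted SphereMask on PixelDepth fades the surface out within FadeDistance of the camera, multiplied by OpacityIn and an inverted vertex-normal Fresnel that fades grazing edges.
# One output adds DepthFade over DepthFadeDistance to soften intersections with scene geometry.
# The other output skips that step to save instructions.
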MaterialFunction /Engine/Functions/Engine_MaterialFunctions01/Opacity/SoftOpacity
graph {
 properties {
  bExposeToLibrary=true
  description="Softens the Opacity input by multiplying against Fresnel, DepthBiasedAlpha and Pixel Depth"
  libraryCategoriesText [Misc,Opacity]
 }
 nodes [
  n1 = PixelDepth()
  n2 = Constant()
  n3 = FunctionInput() {
   bUsePreviewValueAsDefault=true
   description="How close you should get to the surface before it starts fading out DEFAULT=512"
   inputName=FadeDistance
   inputType=FunctionInput_Scalar
   sortPriority=1
   previewValue=<128 0 0 1>
  }
  n4 = FunctionInput() {description=The Opacity Value you want to be Softened, inputName=OpacityIn, inputType=FunctionInput_Scalar, previewValue=<1 1 1 1>}
  n5 = Fresnel(Normal=VertexNormalWS()) {desc="parameterize? ", exponent=1}
  n6 = Multiply(A=OneMinus(None=SphereMask(A=@n1.R, B=@n2, Radius=@n3) {attenuationRadius=512, hardnessPercent=10}), B=Multiply(A=@n4, B=OneMinus(None=@n5)))
  n7 = FunctionInput() {bUsePreviewValueAsDefault=true, inputName=DepthFadeDistance, inputType=FunctionInput_Scalar, previewValue=<64 0 0 0>}
  FunctionOutput(None=DepthFade(Opacity=@n6, FadeDistance=@n7)) {description=Cost 12 more instructions, outputName=OutputUsesDepthBias}
  FunctionOutput(None=@n6) {description=Cost 12 fewer Instructions, outputName=OutputNoDepthBias, sortPriority=1}
 ]
}
