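# Translucent unlit material for the CosmeticTurntable fake shadow, built from the FakeShadowRender render target.
# Eight offset taps blur the target two ways: RGB coverage through SmoothCeil (BlurredEdge) and alpha as inverted, normalized depth (DepthCheck).
# Static switches select between color and depth results or their maximum, with controls for opacity and shadow color.
# A world-space SphereMask around the object position provides a vignette; several unconnected experimental branches remain.
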Material /CosmeticTurntable/Blueprint/FakeShadow/M_FakeShadow
graph {
 properties {bAutomaticallySetUsageInEditor=false, blendMode=BLEND_Translucent, shadingModel=MSM_Unlit, usedShadingModels=Unlit}
 nodes [
  n1 = ScalarParameter() {defaultValue=0.0007, group=_Controls, parameterName=BlurOffset}
  n2 = MaterialFunctionCall!(OffsetAmount=MaterialFunctionCall(X=@n1, Y=@n1) {materialFunction=&MakeFloat2}.Result) {materialFunction=/Engine/Functions/Engine_MaterialFunctions01/ImageAdjustment/BlurSampleOffsets}
  n3 = TextureSample(UVs=Add(A=TextureCoordinate(), B=@n2."0,1"), Tex=NamedRerouteUsage())
  MaterialFunctionCall!(In=MaterialFunctionCall(f3=@n3.RGB) {materialFunction=&AddComponents}.f3, Ramp Width) {materialFunction=/Engine/Functions/Engine_MaterialFunctions01/Math/SmoothCeil}
  n4 = TextureSample(UVs=Add(A=TextureCoordinate(), B=@n2."1,0"), Tex=NamedRerouteUsage())
  MaterialFunctionCall(In=MaterialFunctionCall(f3=@n4.RGB) {materialFunction=&AddComponents}.f3) {materialFunction=&SmoothCeil}
  n5 = TextureSample(UVs=Add(A=TextureCoordinate(), B=@n2."0,-1"), Tex=NamedRerouteUsage())
  MaterialFunctionCall(In=MaterialFunctionCall(f3=@n5.RGB) {materialFunction=&AddComponents}.f3) {materialFunction=&SmoothCeil}
  n6 = TextureSample(UVs=Add(A=TextureCoordinate(), B=@n2."-1,0"), Tex=NamedRerouteUsage())
  MaterialFunctionCall(In=MaterialFunctionCall(f3=@n6.RGB) {materialFunction=&AddComponents}.f3) {materialFunction=&SmoothCeil}
  n7 = TextureSample(UVs=Add(A=TextureCoordinate(), B=@n2."-.5,.5"), Tex=NamedRerouteUsage())
  MaterialFunctionCall(In=MaterialFunctionCall(f3=@n7.RGB) {materialFunction=&AddComponents}.f3) {materialFunction=&SmoothCeil}
  n8 = TextureSample(UVs=Add(A=TextureCoordinate(), B=@n2.".5,.5"), Tex=NamedRerouteUsage())
  MaterialFunctionCall(In=MaterialFunctionCall(f3=@n8.RGB) {materialFunction=&AddComponents}.f3) {materialFunction=&SmoothCeil}
  n9 = TextureSample(UVs=Add(A=TextureCoordinate(), B=@n2.".5,-.5"), Tex=NamedRerouteUsage())
  MaterialFunctionCall(In=MaterialFunctionCall(f3=@n9.RGB) {materialFunction=&AddComponents}.f3) {materialFunction=&SmoothCeil}
  n10 = TextureSample(UVs=Add(A=TextureCoordinate(), B=@n2."-.5,-.5"), Tex=NamedRerouteUsage())
  MaterialFunctionCall(In=MaterialFunctionCall(f3=@n10.RGB) {materialFunction=&AddComponents}.f3) {materialFunction=&SmoothCeil}
  n11 = TextureSampleParameter2D(UVs=Multiply(A=TextureCoordinate(), B=ScalarParameter() {defaultValue=1}) {constB=0.5}) {parameterName=Param, samplerType=SAMPLERTYPE_LinearColor, texture=&/FakeShadowRender}
  n12 = VectorParameter() {defaultValue=<rgba 0.5696 0.7052 0.0001 1000>}
  MaterialFunctionCall!(DepthFadeDistance, OpacityIn=Divide(A=@n11.A, B=@n12.A), FadeDistance) {materialFunction=/Engine/Functions/Engine_MaterialFunctions01/Opacity/SoftOpacity}
  n13 = ScalarParameter() {defaultValue=14000, parameterName=DepthDistance}
  n14 = OneMinus(None=Saturate(None=Divide(A=@n3.A, B=@n13)))
  Multiply(A=@n14) {constB=0.3}
  n15 = ScalarParameter() {defaultValue=14000, parameterName=DepthDistance}
  n16 = OneMinus(None=Saturate(None=Divide(A=@n4.A, B=@n15)))
  Multiply(A=@n16) {constB=0.1}
  n17 = ScalarParameter() {defaultValue=14000, parameterName=DepthDistance}
  n18 = OneMinus(None=Saturate(None=Divide(A=@n5.A, B=@n17)))
  Multiply(A=@n18) {constB=0.2}
  n19 = ScalarParameter() {defaultValue=14000, parameterName=DepthDistance}
  n20 = OneMinus(None=Saturate(None=Divide(A=@n6.A, B=@n19)))
  Multiply(A=@n20) {constB=0.6}
  n21 = ScalarParameter() {defaultValue=14000, parameterName=DepthDistance}
  n22 = OneMinus(None=Saturate(None=Divide(A=@n7.A, B=@n21)))
  Multiply(A=@n22) {constB=0.8}
  n23 = ScalarParameter() {defaultValue=14000, parameterName=DepthDistance}
  n24 = OneMinus(None=Saturate(None=Divide(A=@n8.A, B=@n23)))
  Multiply(A=@n24) {constB=0.5}
  n25 = ScalarParameter() {defaultValue=14000, parameterName=DepthDistance}
  n26 = OneMinus(None=Saturate(None=Divide(A=@n9.A, B=@n25)))
  Multiply(A=@n26) {constB=0.1}
  n27 = ScalarParameter() {defaultValue=14000, parameterName=DepthDistance}
  n28 = OneMinus(None=Saturate(None=Divide(A=@n10.A, B=@n27)))
  Multiply(A=@n28) {constB=0.7}
  n29 = NamedRerouteUsage()
  n30 = NamedRerouteUsage()
  n31 = Max(A=@n29, B=@n30)
  Multiply(
   A=StaticSwitchParameter(True=StaticSwitchParameter(True=@n31, False=@n29) {group=_Controls, parameterName="UseMax(Color,Depth)"}, False=@n30) {group=_Controls, parameterName=UseDepth?}
   B=ScalarParameter() {defaultValue=1, group=_Controls, parameterName=AdjustShadowOpacity, sliderMax=1}
  )
  Multiply(A=SphereMask(
    A=ObjectPositionWS()
    B=WorldPosition().XYZ
    Radius=ScalarParameter() {defaultValue=400, parameterName=VinetteRadius}
    Hardness=ScalarParameter() {parameterName=VinneteHardness}
   ))
  n32 = NamedRerouteDeclaration(None=@n11.RGB) {name=RGB, nodeColor=<rgba 1 0.0469 0 1>}
  n33 = NamedRerouteDeclaration(None=Saturate(None=MaterialFunctionCall(f3=@n32) {materialFunction=&AddComponents}.f3)) {name=CombinedRGB, nodeColor=<rgba 0 0.2734 1 1>}
  Multiply(A=Max(A=Max(A=@n11.R, B=@n11.R), B=@n11.B), B=@n33)
  n34 = TextureObjectParameter() {parameterName=RenderTargetTexture, samplerType=SAMPLERTYPE_LinearColor, texture=&FakeShadowRender}
  NamedRerouteDeclaration(None=@n34) {name=TextureObject, nodeColor=<rgba 0.5937 1 0 1>}
  n35 = TextureCoordinate()
  n36 = ScalarParameter() {defaultValue=0.0007, group=_Controls, parameterName=BlurOffset}
  n37 = MaterialFunctionCall(OffsetAmount=MaterialFunctionCall(X=@n36, Y=@n36) {materialFunction=&MakeFloat2}.Result) {materialFunction=&BlurSampleOffsets}
  n38 = MaterialFunctionCall(f3=TextureSample(UVs=Add(A=@n35, B=@n37."0,1"), Tex=NamedRerouteUsage()).RGB) {materialFunction=&AddComponents}
  n39 = Multiply(A=Saturate(None=MaterialFunctionCall(In=@n38.f3) {materialFunction=&SmoothCeil}.Result)) {constB=0.3}
  n40 = TextureCoordinate()
  n41 = MaterialFunctionCall(f3=TextureSample(UVs=Add(A=@n40, B=@n37."1,0"), Tex=NamedRerouteUsage()).RGB) {materialFunction=&AddComponents}
  n42 = Multiply(A=Saturate(None=MaterialFunctionCall(In=@n41.f3) {materialFunction=&SmoothCeil}.Result)) {constB=0.1}
  n43 = TextureSample(UVs=Add(A=TextureCoordinate(), B=@n37."0,-1"), Tex=NamedRerouteUsage())
  n44 = Saturate(None=MaterialFunctionCall(In=MaterialFunctionCall(f3=@n43.RGB) {materialFunction=&AddComponents}.f3) {materialFunction=&SmoothCeil}.Result)
  n45 = Add(A=TextureCoordinate(), B=@n37."-1,0")
  n46 = NamedRerouteUsage()
  n47 = MaterialFunctionCall(In=MaterialFunctionCall(f3=TextureSample(UVs=@n45, Tex=@n46).RGB) {materialFunction=&AddComponents}.f3) {materialFunction=&SmoothCeil}
  n48 = Add(A=Add(A=Add(A=@n39, B=@n42), B=Multiply(A=@n44) {constB=0.2}), B=Multiply(A=Saturate(None=@n47.Result)) {constB=0.6})
  n49 = TextureCoordinate()
  n50 = MaterialFunctionCall(f3=TextureSample(UVs=Add(A=@n49, B=@n37."-.5,.5"), Tex=NamedRerouteUsage()).RGB) {materialFunction=&AddComponents}
  n51 = Multiply(A=Saturate(None=MaterialFunctionCall(In=@n50.f3) {materialFunction=&SmoothCeil}.Result)) {constB=0.8}
  n52 = TextureSample(UVs=Add(A=TextureCoordinate(), B=@n37.".5,.5"), Tex=NamedRerouteUsage())
  n53 = Saturate(None=MaterialFunctionCall(In=MaterialFunctionCall(f3=@n52.RGB) {materialFunction=&AddComponents}.f3) {materialFunction=&SmoothCeil}.Result)
  n54 = Add(A=TextureCoordinate(), B=@n37.".5,-.5")
  n55 = NamedRerouteUsage()
  n56 = MaterialFunctionCall(In=MaterialFunctionCall(f3=TextureSample(UVs=@n54, Tex=@n55).RGB) {materialFunction=&AddComponents}.f3) {materialFunction=&SmoothCeil}
  n57 = Add(A=Add(A=Add(A=@n48, B=@n51), B=Multiply(A=@n53) {constB=0.5}), B=Multiply(A=Saturate(None=@n56.Result)) {constB=0.1})
  n58 = TextureCoordinate()
  n59 = MaterialFunctionCall(f3=TextureSample(UVs=Add(A=@n58, B=@n37."-.5,-.5"), Tex=NamedRerouteUsage()).RGB) {materialFunction=&AddComponents}
  n60 = Multiply(A=Saturate(None=MaterialFunctionCall(In=@n59.f3) {materialFunction=&SmoothCeil}.Result)) {constB=0.7}
  NamedRerouteDeclaration(None=Divide(A=Add(A=@n57, B=@n60)) {constB=8}) {name=BlurredEdge, nodeColor=<rgba 1 0 0.9141 1>}
  NamedRerouteDeclaration(None=TextureObjectParameter() {parameterName=Param, samplerType=SAMPLERTYPE_LinearColor, texture=&FakeShadowRender}) {name=TextureObject_1, nodeColor=<rgba 0.5938 1 0 1>}
  n61 = Add(A=Add(A=Add(A=@n14, B=@n16), B=@n18), B=@n20)
  n62 = Add(A=Add(A=Add(A=@n61, B=@n22), B=@n24), B=@n26)
  NamedRerouteDeclaration(None=Divide(A=Add(A=@n62, B=@n28)) {constB=8}) {name=DepthCheck, nodeColor=<rgba 1 0 0.4687 1>}
  OneMinus(None=Saturate())
  n63 = Saturate(None=@n33)
  SmoothStep(
   Min=@n12.R
   Max=@n12.R
   Value=Multiply(
    A=Multiply(A=@n11.A, B=@n12.B)
    B=MaterialFunctionCall(In=@n63) {materialFunction=&SmoothCeil}.Result
   )
  ) {constMin=0.3}
  TextureSample(Tex=NamedRerouteUsage())
  VectorParameter() {group=_Controls, parameterName=ShadowColor, defaultValue=<rgba 0 0 0 1>}
 ]
}
